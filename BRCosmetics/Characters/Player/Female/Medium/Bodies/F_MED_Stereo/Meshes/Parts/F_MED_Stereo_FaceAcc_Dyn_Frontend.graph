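# Physics asset for the Stereo female medium character's FaceAcc part, frontend variant, simulating dynamic hair.
# Capsule bodies cover the hair bones plus spine, neck, head, clavicles and upper arms.
# Hair chains with swing and twist limits hang from head or, for side chains continuing into front-shoulder strands, clavicles.
# Cross-constraints tie both back-shoulder chains to the central back chain, while body and master-bone joints rotate freely.
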PhysicsAsset /BRCosmetics/Characters/Player/Female/Medium/Bodies/F_MED_Stereo/Meshes/Parts/F_MED_Stereo_FaceAcc_Dyn_Frontend
bodies {
 clavicle_l [{shapeType=Capsule, length=20.8438, radius=7.2812, shapeName=clavicle_l_capsule, center=<7.957 -0.8251 -1.6594>, rotation=<rot 90 0 0.0001>}]
 clavicle_r [{
   shapeType=Capsule
   length=20.8438
   radius=7.2812
   shapeName=clavicle_l_capsule
   center=<-7.9212 0.5016 1.6594>
   rotation=<rot 90 11.7683 5.4571>
  }]
 dyn_hair_bk_01 [{shapeType=Capsule, length=6.9688, radius=4, shapeName=dyn_hair_bk_01_capsule, center=<3.875 0.0001 2>, rotation=<rot 90 0 0>}]
 dyn_hair_bk_02 [{shapeType=Capsule, length=6.9688, radius=4, shapeName=dyn_hair_bk_01_capsule, center=<3.875 0.0001 2>, rotation=<rot 90 0 0>}]
 dyn_hair_bk_03 [{shapeType=Capsule, length=6.9688, radius=4, shapeName=dyn_hair_bk_01_capsule, center=<3.875 0.0001 1.8>, rotation=<rot 90 0 0>}]
 dyn_hair_bk_04 [{shapeType=Capsule, length=6.9688, radius=4, shapeName=dyn_hair_bk_01_capsule, center=<3.875 0.0001 1.8>, rotation=<rot 90 0 0>}]
 dyn_hair_bk_05 [{
   shapeType=Capsule
   length=6.9688
   radius=4
   shapeName=dyn_hair_bk_01_capsule
   center=<3.5419 0.0001 0.1828>
   rotation=<rot 82.9494 -0.0001 -0.0001>
  }]
 dyn_hair_bk_06 [{shapeType=Capsule, length=6.9688, radius=4, shapeName=dyn_hair_bk_01_capsule, center=<3.9127 0.0001 0.5491>, rotation=<rot 86.552 0 0>}]
 dyn_hair_bk_MSTR [{shapeType=Capsule, length=0.5, radius=0.5, shapeName=dyn_hair_bk_MSTR_capsule}]
 dyn_hair_bk_sh_l_00 [{
   shapeType=Capsule
   length=14
   radius=4.5
   shapeName=dyn_hair_bk_01_capsule
   center=<7.5105 0.0003 2.426>
   rotation=<rot 79.3417 -0.0001 -0.0001>
  }]
 dyn_hair_bk_sh_l_01 [{shapeType=Capsule, length=6.9688, radius=4, shapeName=dyn_hair_bk_01_capsule, center=<3.875 0.0004 1.75>, rotation=<rot 90 0 0>}]
 dyn_hair_bk_sh_l_02 [{shapeType=Capsule, length=6.9688, radius=4, shapeName=dyn_hair_bk_01_capsule, center=<3.875 0.0004 2.3>, rotation=<rot 90 0 0>}]
 dyn_hair_bk_sh_l_03 [{
   shapeType=Capsule
   length=6.9688
   radius=4
   shapeName=dyn_hair_bk_01_capsule
   center=<5.1096 0.0004 -1.027>
   rotation=<rot 74.2239 -0.0001 -0.0001>
  }]
 dyn_hair_bk_sh_r_00 [
  {shapeType=Capsule, length=14.875, radius=3.625, shapeName=dyn_hair_bk_01_capsule, center=<8.029 0 3.3647>, rotation=<rot 90 0 0>}
  {
   shapeType=Capsule
   length=1.375
   radius=5.0938
   shapeName=dyn_hair_bk_01_capsule
   center=<12.2976 -0.5851 2.7308>
   rotation=<rot 56.689 129.4549 125.7344>
  }
 ]
 dyn_hair_bk_sh_r_01 [{shapeType=Capsule, length=6.9688, radius=4, shapeName=dyn_hair_bk_01_capsule, center=<3.875 0.0004 1.75>, rotation=<rot 90 0 0>}]
 dyn_hair_bk_sh_r_02 [{shapeType=Capsule, length=6.9688, radius=4, shapeName=dyn_hair_bk_01_capsule, center=<3.875 0.0004 2.3>, rotation=<rot 90 0 0>}]
 dyn_hair_bk_sh_r_03 [{shapeType=Capsule, length=6.9688, radius=4, shapeName=dyn_hair_bk_01_capsule, center=<3.4169 0 1.1408>, rotation=<rot 68.4379 0 0>}]
 dyn_hair_fr_l_01 [{shapeType=Capsule, length=5.9688, radius=5, shapeName=dyn_hair_bk_01_capsule, center=<6.3993 0.0002 4>, rotation=<rot 81.5082 0.0002 0.0001>}]
 dyn_hair_fr_l_02 [{shapeType=Capsule, length=5.375, radius=5.5938, shapeName=dyn_hair_bk_01_capsule, center=<6.3993 0.0002 4>, rotation=<rot 90 0 0>}]
 dyn_hair_fr_l_MSTR [{shapeType=Capsule, length=0.5, radius=0.5, shapeName=dyn_hair_fr_l_MSTR_capsule}]
 dyn_hair_fr_r_01 [{shapeType=Capsule, length=5.8125, radius=5.1562, shapeName=dyn_hair_bk_01_capsule, center=<6.3993 0.0002 4>, rotation=<rot 79.2371 0 0>}]
 dyn_hair_fr_r_02 [{shapeType=Capsule, length=4.2812, radius=6.6875, shapeName=dyn_hair_bk_01_capsule, center=<6.3993 0.0002 4>, rotation=<rot 90 0 0>}]
 dyn_hair_fr_r_MSTR [{shapeType=Capsule, length=0.5, radius=0.5, shapeName=dyn_hair_fr_r_MSTR_capsule}]
 dyn_hair_fr_sh_l_01 [{shapeType=Capsule, length=5, radius=1.8125, shapeName=dyn_hair_bk_01_capsule, center=<5.3347 -0.6 0.0001>, rotation=<rot 90 0 0>}]
 dyn_hair_fr_sh_l_02 [{
   shapeType=Capsule
   length=4.4688
   radius=2.1562
   shapeName=dyn_hair_bk_01_capsule
   center=<3.477 0 0.0257>
   rotation=<rot 86.3027 0.0001 0.0001>
  }]
 dyn_hair_fr_sh_l_03 [{shapeType=Capsule, length=3.3438, radius=3.0938, shapeName=dyn_hair_bk_01_capsule, center=<2.0571 0 0.0001>, rotation=<rot 90 0 0>}]
 dyn_hair_fr_sh_l_04 [{shapeType=Capsule, length=4.1875, radius=1.9062, shapeName=dyn_hair_bk_01_capsule, center=<2.2735 0 0>, rotation=<rot 90 0 0>}]
 dyn_hair_fr_sh_r_01 [{shapeType=Capsule, length=5, radius=1.8125, shapeName=dyn_hair_bk_01_capsule, center=<4.7166 1.543 1.9999>, rotation=<rot 90 0 0>}]
 dyn_hair_fr_sh_r_02 [{shapeType=Capsule, length=4.1875, radius=2.5312, shapeName=dyn_hair_bk_01_capsule, center=<2.6531 0 -0.6701>, rotation=<rot 87.1042 0 0>}]
 dyn_hair_fr_sh_r_03 [{
   shapeType=Capsule
   length=2.875
   radius=2.2812
   shapeName=dyn_hair_bk_01_capsule
   center=<2.239 0 0.8367>
   rotation=<rot 84.833 0.0001 0.0001>
  }]
 dyn_hair_l_01 [{shapeType=Capsule, length=6.9688, radius=4, shapeName=dyn_hair_bk_01_capsule, center=<2.6382 -3.4 0>, rotation=<rot 90 0 0>}]
 dyn_hair_l_02 [{shapeType=Capsule, length=5, radius=2.9375, shapeName=dyn_hair_bk_01_capsule, center=<4.1401 -2.2644 2.212>, rotation=<rot 69.9866 180 180>}]
 dyn_hair_l_MSTR [{shapeType=Capsule, length=0.5, radius=0.5, shapeName=dyn_hair_l_MSTR_capsule}]
 dyn_hair_r_01 [{shapeType=Capsule, length=6.9688, radius=4, shapeName=dyn_hair_bk_01_capsule, center=<2.6382 3.4 0>, rotation=<rot 90 0 0>}]
 dyn_hair_r_02 [{shapeType=Capsule, length=8, radius=2.5, shapeName=dyn_hair_bk_01_capsule, center=<4.6755 2.26 0.742>, rotation=<rot 69.9866 180 180>}]
 dyn_hair_r_MSTR [{shapeType=Capsule, length=0.5, radius=0.5, shapeName=dyn_hair_r_MSTR_capsule}]
 head [{shapeType=Capsule, length=7.2188, radius=6.75, shapeName=spine_05_capsule, center=<0.8645 0.5573 0>, rotation=<rot 90 0 -8.2538>}]
 neck_01 [
  {shapeType=Capsule, length=7.6875, radius=5.0625, shapeName=spine_05_capsule, center=<0.8645 0.5573 0>, rotation=<rot 90 0 0>}
  {shapeType=Capsule, length=25.25, radius=4.75, shapeName=neck_01_capsule, center=<-0.8497 -0.7331 0>}
 ]
 neck_02 [{shapeType=Capsule, length=30, radius=15, shapeName=neck_02_capsule, center=<-5.2395 -12.1231 0>}]
 spine_03 [{shapeType=Capsule, length=4.0938, radius=10.125, shapeName=spine_05_capsule, center=<4.6909 2.2638 0>}]
 spine_04 [{shapeType=Capsule, length=10.6562, radius=10.25, shapeName=spine_05_capsule, center=<-0.0656 3.1574 0>}]
 spine_05 [{shapeType=Capsule, length=10.0312, radius=8.3125, shapeName=spine_05_capsule, center=<0.9453 -0.5742 0>}]
 upperarm_l [{shapeType=Capsule, length=19.7812, radius=8, shapeName=clavicle_l_capsule, center=<11.381 0.0001 0>, rotation=<rot 90 0 0>}]
 upperarm_r [{shapeType=Capsule, length=19.7812, radius=7, shapeName=clavicle_l_capsule, center=<-11.381 -0.0001 0>, rotation=<rot 90 0 180>}]
}
constraints [
 {bone1Name=dyn_hair_l_01, bone2Name=dyn_hair_l_MSTR, swing1LimitDegrees=15, swing2LimitDegrees=20, twistLimitDegrees=5}
 {bone1Name=dyn_hair_l_02, bone2Name=dyn_hair_l_01, swing1LimitDegrees=25, swing2LimitDegrees=20, twistLimitDegrees=5}
 {bone1Name=dyn_hair_fr_sh_l_01, bone2Name=dyn_hair_l_02, swing1LimitDegrees=25, swing2LimitDegrees=20, twistLimitDegrees=5}
 {bone1Name=dyn_hair_fr_sh_l_02, bone2Name=dyn_hair_fr_sh_l_01, swing1LimitDegrees=15, swing2LimitDegrees=45, twistLimitDegrees=5}
 {bone1Name=dyn_hair_fr_sh_l_03, bone2Name=dyn_hair_fr_sh_l_02, swing1LimitDegrees=15, swing2LimitDegrees=45, twistLimitDegrees=5}
 {bone1Name=dyn_hair_fr_sh_l_04, bone2Name=dyn_hair_fr_sh_l_02, swing1LimitDegrees=15, swing2LimitDegrees=45, twistLimitDegrees=5}
 {bone1Name=dyn_hair_bk_sh_l_01, bone2Name=dyn_hair_bk_sh_l_00, swing1LimitDegrees=15, swing2LimitDegrees=45, twistLimitDegrees=5}
 {bone1Name=dyn_hair_bk_sh_l_02, bone2Name=dyn_hair_bk_sh_l_01, swing1LimitDegrees=15, swing2LimitDegrees=45, twistLimitDegrees=5}
 {bone1Name=dyn_hair_bk_sh_l_03, bone2Name=dyn_hair_bk_sh_l_02, swing1LimitDegrees=15, swing2LimitDegrees=45, twistLimitDegrees=5}
 {bone1Name=dyn_hair_r_01, bone2Name=dyn_hair_r_MSTR, swing1LimitDegrees=15, swing2LimitDegrees=20, twistLimitDegrees=5}
 {bone1Name=dyn_hair_r_02, bone2Name=dyn_hair_r_01, swing1LimitDegrees=25, swing2LimitDegrees=20, twistLimitDegrees=5}
 {bone1Name=dyn_hair_bk_sh_r_01, bone2Name=dyn_hair_bk_sh_r_00, swing1LimitDegrees=25, swing2LimitDegrees=45, twistLimitDegrees=5}
 {bone1Name=dyn_hair_bk_sh_r_02, bone2Name=dyn_hair_bk_sh_r_01, swing1LimitDegrees=30, swing2LimitDegrees=45, twistLimitDegrees=5}
 {bone1Name=dyn_hair_bk_sh_r_03, bone2Name=dyn_hair_bk_sh_r_02, swing1LimitDegrees=15, swing2LimitDegrees=45, twistLimitDegrees=5}
 {bone1Name=dyn_hair_fr_sh_r_01, bone2Name=dyn_hair_r_02, swing1LimitDegrees=25, swing2LimitDegrees=20, twistLimitDegrees=5}
 {bone1Name=dyn_hair_fr_sh_r_02, bone2Name=dyn_hair_fr_sh_r_01, swing1LimitDegrees=15, swing2LimitDegrees=45, twistLimitDegrees=5}
 {bone1Name=dyn_hair_fr_sh_r_03, bone2Name=dyn_hair_fr_sh_r_01, swing1LimitDegrees=15, swing2LimitDegrees=45, twistLimitDegrees=5}
 {bone1Name=dyn_hair_fr_l_01, bone2Name=dyn_hair_fr_l_MSTR, swing1LimitDegrees=25, swing2LimitDegrees=15, twistLimitDegrees=5}
 {bone1Name=dyn_hair_fr_l_02, bone2Name=dyn_hair_fr_l_01, swing1LimitDegrees=25, swing2LimitDegrees=15, twistLimitDegrees=5}
 {bone1Name=dyn_hair_bk_01, bone2Name=dyn_hair_bk_MSTR, swing1LimitDegrees=15, swing2LimitDegrees=45, twistLimitDegrees=5}
 {bone1Name=dyn_hair_bk_02, bone2Name=dyn_hair_bk_01, swing1LimitDegrees=15, swing2LimitDegrees=45, twistLimitDegrees=5}
 {bone1Name=dyn_hair_bk_03, bone2Name=dyn_hair_bk_02, swing1LimitDegrees=15, swing2LimitDegrees=45, twistLimitDegrees=5}
 {bone1Name=dyn_hair_bk_04, bone2Name=dyn_hair_bk_03, swing1LimitDegrees=15, swing2LimitDegrees=45, twistLimitDegrees=5}
 {bone1Name=dyn_hair_bk_05, bone2Name=dyn_hair_bk_04, swing1LimitDegrees=15, swing2LimitDegrees=45, twistLimitDegrees=5}
 {bone1Name=dyn_hair_bk_06, bone2Name=dyn_hair_bk_04, swing1LimitDegrees=15, swing2LimitDegrees=45, twistLimitDegrees=5}
 {bone1Name=dyn_hair_fr_r_01, bone2Name=dyn_hair_fr_r_MSTR, swing1LimitDegrees=25, swing2LimitDegrees=15, twistLimitDegrees=5}
 {bone1Name=dyn_hair_fr_r_02, bone2Name=dyn_hair_fr_r_01, swing1LimitDegrees=25, swing2LimitDegrees=15, twistLimitDegrees=5}
 {
  bone1Name=dyn_hair_fr_l_MSTR
  bone2Name=head
  swing1LimitDegrees=45
  swing1Motion=Free
  swing2LimitDegrees=45
  swing2Motion=Free
  twistLimitDegrees=45
  twistMotion=Free
 }
 {
  bone1Name=dyn_hair_bk_MSTR
  bone2Name=head
  swing1LimitDegrees=45
  swing1Motion=Free
  swing2LimitDegrees=45
  swing2Motion=Free
  twistLimitDegrees=45
  twistMotion=Free
 }
 {
  bone1Name=dyn_hair_fr_r_MSTR
  bone2Name=head
  swing1LimitDegrees=45
  swing1Motion=Free
  swing2LimitDegrees=45
  swing2Motion=Free
  twistLimitDegrees=45
  twistMotion=Free
 }
 {
  bone1Name=head
  bone2Name=neck_01
  swing1LimitDegrees=45
  swing1Motion=Free
  swing2LimitDegrees=45
  swing2Motion=Free
  twistLimitDegrees=45
  twistMotion=Free
 }
 {
  bone1Name=dyn_hair_r_MSTR
  bone2Name=clavicle_r
  swing1LimitDegrees=45
  swing1Motion=Free
  swing2LimitDegrees=45
  swing2Motion=Free
  twistLimitDegrees=45
  twistMotion=Free
 }
 {
  bone1Name=dyn_hair_l_MSTR
  bone2Name=clavicle_l
  swing1LimitDegrees=45
  swing1Motion=Free
  swing2LimitDegrees=45
  swing2Motion=Free
  twistLimitDegrees=45
  twistMotion=Free
 }
 {
  bone1Name=neck_01
  bone2Name=spine_05
  swing1LimitDegrees=45
  swing1Motion=Free
  swing2LimitDegrees=45
  swing2Motion=Free
  twistLimitDegrees=45
  twistMotion=Free
 }
 {
  bone1Name=spine_05
  bone2Name=spine_04
  swing1LimitDegrees=45
  swing1Motion=Free
  swing2LimitDegrees=45
  swing2Motion=Free
  twistLimitDegrees=45
  twistMotion=Free
 }
 {
  bone1Name=spine_04
  bone2Name=spine_03
  swing1LimitDegrees=45
  swing1Motion=Free
  swing2LimitDegrees=45
  swing2Motion=Free
  twistLimitDegrees=45
  twistMotion=Free
 }
 {
  bone1Name=clavicle_r
  bone2Name=spine_05
  swing1LimitDegrees=45
  swing1Motion=Free
  swing2LimitDegrees=45
  swing2Motion=Free
  twistLimitDegrees=45
  twistMotion=Free
 }
 {bone1Name=dyn_hair_bk_sh_l_02, bone2Name=dyn_hair_bk_04, swing1LimitDegrees=45, swing2LimitDegrees=45, twistLimitDegrees=45}
 {bone1Name=dyn_hair_bk_sh_r_02, bone2Name=dyn_hair_bk_04, swing1LimitDegrees=45, swing2LimitDegrees=45, twistLimitDegrees=45}
 {bone1Name=dyn_hair_bk_sh_l_00, bone2Name=head, swing1LimitDegrees=20, swing2LimitDegrees=50, twistLimitDegrees=5}
 {bone1Name=dyn_hair_bk_sh_r_00, bone2Name=head, swing1LimitDegrees=20, swing2LimitDegrees=40, twistLimitDegrees=5}
 {
  bone1Name=clavicle_l
  bone2Name=spine_05
  swing1LimitDegrees=45
  swing1Motion=Free
  swing2LimitDegrees=45
  swing2Motion=Free
  twistLimitDegrees=45
  twistMotion=Free
 }
 {
  bone1Name=upperarm_l
  bone2Name=clavicle_l
  swing1LimitDegrees=45
  swing1Motion=Free
  swing2LimitDegrees=45
  swing2Motion=Free
  twistLimitDegrees=45
  twistMotion=Free
 }
 {
  bone1Name=upperarm_r
  bone2Name=clavicle_r
  swing1LimitDegrees=45
  swing1Motion=Free
  swing2LimitDegrees=45
  swing2Motion=Free
  twistLimitDegrees=45
  twistMotion=Free
 }
 {
  bone1Name=dyn_hair_bk_05
  bone2Name=dyn_hair_bk_sh_l_03
  swing1LimitDegrees=45
  swing1Motion=Free
  swing2LimitDegrees=45
  swing2Motion=Free
  twistLimitDegrees=45
  twistMotion=Free
 }
]
