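# Anim Blueprint for the F_MED_Stereo face-accessory part, copying the attached parent's pose and simulating hair.
# RigidBody nodes use separate front-end and in-game physics assets; with physics disabled, a static path adds front-end bone offsets instead.
# Both paths scale down three back hair bones when a backbling flag is set.
# Updates adjust gravity and sim-space settings for skydiving or front end; dynamics reset when skydiving ends.
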AnimBlueprint /BRCosmetics/Characters/Player/Female/Medium/Bodies/F_MED_Stereo/Meshes/Parts/F_MED_Stereo_FaceAcc_AnimBP
target_skeleton=&/F_MED_Stereo_FaceAcc_Skeleton
graphs [
 AnimGraph {nodes [
   "Transform (Modify) Bone - Bone: dyn_hair_fr_sh_l_01"!(ComponentPose, bAlphaBoolEnabled:bool=Get bIsInFrontEnd(), Translation="0,0,0", Rotation="-19.917,5.319,23.183") -> (Pose) {
    alphaInputType=Bool
    rotationMode=BMM_Additive
    rotationSpace=BCS_BoneSpace
    scaleSpace=BCS_BoneSpace
    alphaBoolBlend {blendOption=Linear}
    boneToModify {boneName=dyn_hair_fr_sh_l_01}
    rotation=<rot -19.917 5.319 23.1828>
   }
   n1 = Get isWearingBackbling()
   n2 = "Transform (Modify) Bone - Bone: dyn_hair_bk_04"!(
    ComponentPose=Local To Component(LocalPose=Copy Pose From Mesh() {bUseAttachedParent=true})
    bAlphaBoolEnabled:bool=@n1
    Translation="0,0,0"
    Rotation="0,0,0"
   ) -> (Pose) {
    alphaInputType=Bool
    scaleMode=BMM_Additive
    scaleSpace=BCS_BoneSpace
    alphaBoolBlend {blendOption=Linear}
    boneToModify {boneName=dyn_hair_bk_04}
    scale=<0.3 0.3 0.3>
   }
   n3 = "Transform (Modify) Bone - Bone: dyn_hair_bk_sh_l_02"!(ComponentPose=@n2, bAlphaBoolEnabled:bool=@n1, Translation="0,0,0", Rotation="0,0,0") -> (Pose) {
    alphaInputType=Bool
    scaleMode=BMM_Additive
    scaleSpace=BCS_BoneSpace
    alphaBoolBlend {blendOption=Linear}
    boneToModify {boneName=dyn_hair_bk_sh_l_02}
    scale=<0.3 0.3 0.3>
   }
   n4 = "Transform (Modify) Bone - Bone: dyn_hair_bk_sh_r_02"!(ComponentPose=@n3, bAlphaBoolEnabled:bool=@n1, Translation="0,0,0", Rotation="0,0,0") -> (Pose) {
    alphaInputType=Bool
    scaleMode=BMM_Additive
    scaleSpace=BCS_BoneSpace
    alphaBoolBlend {blendOption=Linear}
    boneToModify {boneName=dyn_hair_bk_sh_r_02}
    scale=<0.3 0.3 0.3>
   }
   n5 = Get bIsInFrontEnd()
   n6 = "Transform (Modify) Bone - Bone: spine_05"!(ComponentPose=@n4, bAlphaBoolEnabled:bool=@n5, Translation="0,0,0", Rotation="0,0,0") -> (Pose) {
    alphaInputType=Bool
    rotationMode=BMM_Additive
    rotationSpace=BCS_BoneSpace
    scaleSpace=BCS_BoneSpace
    alphaBoolBlend {blendOption=Linear}
    boneToModify {boneName=spine_05}
   }
   n7 = "Transform (Modify) Bone - Bone: neck_01"!(ComponentPose=@n6, bAlphaBoolEnabled:bool=@n5, Translation="0,0,0", Rotation="0,0,0") -> (Pose) {
    alphaInputType=Bool
    rotationMode=BMM_Additive
    rotationSpace=BCS_BoneSpace
    scaleSpace=BCS_BoneSpace
    alphaBoolBlend {blendOption=Linear}
    boneToModify {boneName=neck_01}
   }
   n8 = "Transform (Modify) Bone - Bone: dyn_hair_bk_sh_r_00"!(ComponentPose=@n7, bAlphaBoolEnabled:bool=@n5, Translation="0,0,0", Rotation="0,0,0") -> (Pose) {
    alphaInputType=Bool
    rotationMode=BMM_Additive
    rotationSpace=BCS_BoneSpace
    scaleSpace=BCS_BoneSpace
    alphaBoolBlend {blendOption=Linear}
    boneToModify {boneName=dyn_hair_bk_sh_r_00}
   }
   n9 = "Transform (Modify) Bone - Bone: dyn_hair_bk_02"!(ComponentPose=@n8, bAlphaBoolEnabled:bool=@n5, Translation="0,0,0", Rotation="0.000,15.001,0.000") -> (Pose) {
    alphaInputType=Bool
    rotationMode=BMM_Additive
    rotationSpace=BCS_BoneSpace
    scaleSpace=BCS_BoneSpace
    alphaBoolBlend {blendOption=Linear}
    boneToModify {boneName=dyn_hair_bk_02}
    rotation {yaw=15.0014}
   }
   n10 = "Transform (Modify) Bone - Bone: dyn_hair_bk_sh_l_00"!(ComponentPose=@n9, bAlphaBoolEnabled:bool=@n5, Translation="0,0,0", Rotation="-19.921,15.319,-1.817") -> (Pose) {
    alphaInputType=Bool
    rotationMode=BMM_Additive
    rotationSpace=BCS_BoneSpace
    scaleSpace=BCS_BoneSpace
    alphaBoolBlend {blendOption=Linear}
    boneToModify {boneName=dyn_hair_bk_sh_l_00}
    rotation=<rot -19.921 15.3194 -1.817>
   }
   n11 = RigidBody!(
    ComponentPose=Local To Component(LocalPose=Copy Pose From Mesh() {bUseAttachedParent=true})
    Alpha
    OverrideWorldGravity:Vector=Get GravityOverride()
    ExternalForce
    SimSpaceSettings:Sim Space Settings Structure=Get Sim Space Settings() -> (Sim Space Settings:Sim Space Settings Structure)
   ) -> (Pose) {
    bClampLinearTranslationLimitToRefPose=true
    bOverrideWorldGravity=true
    overridePhysicsAsset=&/F_MED_Stereo_FaceAcc_Dyn
    simulationSpace=BaseBoneSpace
    alphaBoolBlend {blendOption=Linear}
    baseBoneRef {boneName=neck_01}
    componentAppliedLinearAccClamp=<3000 3000 3000>
    componentLinearAccScale=<0.3 0.15>
    componentLinearVelScale=<0.3 0.15>
    overrideWorldGravity {z=-980}
    simSpaceSettings {
     maxAngularAcceleration=1000
     maxAngularVelocity=0.3
     maxLinearAcceleration=700
     maxLinearVelocity=500
     velocityScaleZ=0
     worldAlpha=0.5
     externalLinearDragV=<0.5 0.5>
    }
   }
   n12 = Component To Local(ComponentPose=@n11)
   n13 = RigidBody(
    ComponentPose=Local To Component(LocalPose=Copy Pose From Mesh() {bUseAttachedParent=true})
    OverrideWorldGravity=Get GravityOverride()
    SimSpaceSettings=Get Sim Space Settings()
   ) {
    bClampLinearTranslationLimitToRefPose=true
    bOverrideWorldGravity=true
    overridePhysicsAsset=&/F_MED_Stereo_FaceAcc_Dyn_Frontend
    simulationSpace=BaseBoneSpace
    alphaBoolBlend {blendOption=Linear}
    baseBoneRef {boneName=neck_01}
    componentAppliedLinearAccClamp=<3000 3000 3000>
    componentLinearAccScale=<0.3 0.15>
    componentLinearVelScale=<0.3 0.15>
    overrideWorldGravity {z=-980}
    simSpaceSettings {
     maxAngularAcceleration=1000
     maxAngularVelocity=0.3
     maxLinearAcceleration=700
     maxLinearVelocity=500
     velocityScaleZ=0
     worldAlpha=0.5
     externalLinearDragV=<0.5 0.5>
    }
   }
   n14 = Component To Local(ComponentPose=@n13)
   n15 = Get bIsInFrontEnd()
   n16 = Get isWearingBackbling()
   n17 = "Transform (Modify) Bone - Bone: dyn_hair_bk_04"(
    ComponentPose=Local To Component(LocalPose=Blend Poses by bool(BlendPose_0=@n14, BlendPose_1=@n12, bActiveValue=@n15) {blendType=Linear})
    bAlphaBoolEnabled=@n16
    Translation="0,0,0"
    Rotation="0,0,0"
   ) {
    alphaInputType=Bool
    scaleMode=BMM_Additive
    scaleSpace=BCS_BoneSpace
    alphaBoolBlend {blendOption=Linear}
    boneToModify {boneName=dyn_hair_bk_04}
    scale=<0.3 0.3 0.3>
   }
   n18 = "Transform (Modify) Bone - Bone: dyn_hair_bk_sh_l_02"(ComponentPose=@n17, bAlphaBoolEnabled=@n16, Translation="0,0,0", Rotation="0,0,0") {
    alphaInputType=Bool
    scaleMode=BMM_Additive
    scaleSpace=BCS_BoneSpace
    alphaBoolBlend {blendOption=Linear}
    boneToModify {boneName=dyn_hair_bk_sh_l_02}
    scale=<0.3 0.3 0.3>
   }
   n19 = "Transform (Modify) Bone - Bone: dyn_hair_bk_sh_r_02"(ComponentPose=@n18, bAlphaBoolEnabled=@n16, Translation="0,0,0", Rotation="0,0,0") {
    alphaInputType=Bool
    scaleMode=BMM_Additive
    scaleSpace=BCS_BoneSpace
    alphaBoolBlend {blendOption=Linear}
    boneToModify {boneName=dyn_hair_bk_sh_r_02}
    scale=<0.3 0.3 0.3>
   }
   n20 = Blend Poses by bool(BlendPose_0=Component To Local(ComponentPose=@n19), BlendPose_1=Component To Local(ComponentPose=@n10), bActiveValue=Get bRigidBodyEnabled()) {blendType=Linear}
   Output Pose(Result=Component To Local(ComponentPose=Local To Component(LocalPose=@n20))) {layerGroup=None}
  ]}
 EventGraph {nodes [
   Event BlueprintUpdateAnimation()
   > n1 = Sequence()
   > n2 = Gravity Override State Function DEPRECATED(CharacterPartAnimInstance=Self-Reference(), GameTimeInSeconds=GetGameTimeInSeconds(), GravityOverrideParams=Get Gravity Override Params())
   > Set GravityOverride(GravityOverride=@n2)
   n3 = Branch(execute=@n1.then_1, Condition=Get bIsSkyDiving())
   > Set Sim Space Settings(Sim Space Settings:Sim Space Settings Structure=Make Sim Space Settings!(
     WorldAlpha:float=0
     VelocityScaleZ:float=0
     DampingAlpha:float=1
     MaxLinearVelocity:float=500
     MaxAngularVelocity:float=0.3
     MaxLinearAcceleration:float=700
     MaxAngularAcceleration:float=1000
     ExternalLinearDragV:Vector="0.5,0.5,0"
     ExternalLinearVelocity:Vector="0,0,0"
     ExternalAngularVelocity:Vector="0,0,0"
    ) -> (SimSpaceSettings:Sim Space Settings Structure) {is=title K2Node_MakeStruct}) -> (Output_Get:Sim Space Settings Structure)
   n4 = Branch(execute=@n3.else, Condition=Get bIsInFrontEnd())
   > Set Sim Space Settings(Sim Space Settings=Make Sim Space Settings(
     WorldAlpha=0.1
     VelocityScaleZ=0
     DampingAlpha=1
     MaxLinearVelocity=500
     MaxAngularVelocity=0.3
     MaxLinearAcceleration=700
     MaxAngularAcceleration=1000
     ExternalLinearDragV="0.5,0.5,0"
     ExternalLinearVelocity="0,0,0"
     ExternalAngularVelocity="0,0,0"
    ))
   Set Sim Space Settings(
    execute=@n4.else
    Sim Space Settings=Make Sim Space Settings(
     WorldAlpha=0.5
     VelocityScaleZ=0
     DampingAlpha=1
     MaxLinearVelocity=500
     MaxAngularVelocity=0.3
     MaxLinearAcceleration=700
     MaxAngularAcceleration=1000
     ExternalLinearDragV="0.5,0.5,0"
     ExternalLinearVelocity="0,0,0"
     ExternalAngularVelocity="0,0,0"
    )
   )
   Event OnEndSkydiving()
   > ResetDynamics(InTeleportType=TeleportPhysics)
   > ResetDynamics(InTeleportType=ResetPhysics)
   Event OnCharacterCustomizationLoaded()
   n5 = CheckForBackbling()
   > Set isWearingBackbling(isWearingBackbling=OR Boolean!(A=@n5, B=@n5.bIsBackblingACape, C:bool=@n5.bIsBackblingLong) -> (ReturnValue))
   GetOwnerSkeletalMeshComponent()
   GetOwningActor()
  ]}
]
variables [
 Gravity Override Params {
  type=GravityOverrideParamsStruct
  default {
   skydive_FloatingDown {gravityOverride=<0 -2000 -300>, windAmplitude=<200 200 200>, windOffset=<0 -600 200>}
   skydive_GoingUp {gravityOverride=<0 -2000 300>}
   skydive_Parachuting {gravityOverride=<400 -400 -600>, windAmplitude=<200 100 100>}
  }
 }
 GravityOverride {type=Vector}
 isWearingBackbling {type=boolean}
 Sim Space Settings {
  type=SimSpaceSettings
  default {
   dampingAlpha=1
   maxAngularAcceleration=1000
   maxAngularVelocity=0.3
   maxLinearAcceleration=700
   maxLinearVelocity=500
   velocityScaleZ=0
   worldAlpha=0.5
   externalLinearDragV=<0.5 0.5 0>
  }
 }
]
